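; PCG example graph that builds a forest on a square spline area, named as 200m with medium trees and undergrowth, plus a planar mesh.
; Two random fills scatter points for separate asset spawns, one providing the trees that anchor seedlings.
; Circles around those trees are unioned for faster processing, then filled uniformly.
; Distance from trees shrinks seedlings, overlap filtering prunes points, and random position offsets at spawn hide the grid.
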(graph Forest_Square_200m_with_Medium_Trees_and_undergrowth
  (bind Subgraph_13 (Subgraph :Graph /PCGPrimitives/Primitives/Create/Create_Spline_Shape.Create_Spline_Shape))
  (bind Subgraph_15 (Subgraph :Graph /PCGPrimitives/Primitives/Fill/Fill_Areas_Randomly.Fill_Areas_Randomly))
  (bind Subgraph_8 (Subgraph :Graph /PCGPrimitives/Primitives/Create/Create_Spline_Shape.Create_Spline_Shape))
  (bind Subgraph_1 (Subgraph :Graph /PCGPrimitives/Primitives/Compose/Compose_Areas_Union.Compose_Areas_Union)) ; HEre we union all the circles areas around the \nlarge trees to treat the spawning area as one,\nMaking it a lot faster to process for PCG.
  (bind Subgraph_9 (Subgraph :Graph /PCGPrimitives/Primitives/Fill/Fill_Areas_Uniformly.Fill_Areas_Uniformly))
  (bind Subgraph_11 (Subgraph :Graph /PCGPrimitives/Primitives/Write/Write_DistanceFrom.Write_DistanceFrom)) ; The further away from the tree, the smaller the seedling will be. 
  (bind Subgraph_10 (Subgraph :Graph /PCGPrimitives/Primitives/Filter/Filter_Points_Overlapping.Filter_Points_Overlapping))
  (bind Subgraph_12 (Subgraph :Graph /PCGPrimitives/Primitives/Spawn/Spawn_Assets.Spawn_Assets)) ; Notice the random position paramter. Since we have used a Fill areas uniformly upstream. \nWe need to add a bit of randomness in the position here, if not \nAll the seedlings will be placed on a grid. Which wouldn't look realistic.
  (bind Subgraph_14 (Subgraph :Graph /PCGPrimitives/Primitives/Fill/Fill_Areas_Randomly.Fill_Areas_Randomly))
  (bind Subgraph_16 (Subgraph :Graph /PCGPrimitives/Primitives/Spawn/Spawn_Assets.Spawn_Assets))
  (bind Subgraph_17 (Subgraph :Graph /PCGPrimitives/Primitives/Create/Create_Mesh_Planar.Create_Mesh_Planar))
  (bind Subgraph_18 (Subgraph :Graph /PCGPrimitives/Primitives/Spawn/Spawn_Assets.Spawn_Assets))
  (connect Subgraph_8 Out Subgraph_1 PrimaryInput)
  (connect Subgraph_15 Out Subgraph_8 PrimaryInput)
  (connect Subgraph_1 Out Subgraph_9 PrimaryInput)
  (connect Subgraph_11 Out Subgraph_10 PrimaryInput)
  (connect Subgraph_9 Out Subgraph_11 PrimaryInput)
  (connect Subgraph_15 Out Subgraph_11 SecondaryInput)
  (connect Subgraph_10 Out Subgraph_12 PrimaryInput)
  (connect Subgraph_13 Out Subgraph_14 PrimaryInput)
  (connect Subgraph_13 Out Subgraph_15 PrimaryInput)
  (connect Subgraph_15 Out Subgraph_16 PrimaryInput)
  (connect Subgraph_13 Out Subgraph_17 PrimaryInput)
  (connect Subgraph_14 Out Subgraph_18 PrimaryInput))
